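# Niagara effect for a Flopper fishing spot, showing swirling bubbles and fish below the system origin.
# GPU emitters spawn bubble meshes in a wandering vortex and a dense stream of short-lived fish meshes driven by attraction, vortices and curl noise.
# A third emitter bursts persistent fish in two spawn groups that swim toward computed tracking points.
# Their speed and smoothed turning feed material parameters, alongside per-group variation and time-of-day scaling.
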NiagaraSystem /Game/Effects/Fort_Effects/Gameplay/Floppers/NS_Flopper_Spot
dependencies {
 usedDataInterfaces [&NiagaraDataInterfaceCurve, &NiagaraDataInterfaceArrayFloat]
 usedDynamicInputs [&SimulationPosition]
 usedModules [/Niagara/Modules/System/SystemLifeCycle, &SystemState]
}
emitters {
 Bubbles001 {
  simTarget=GPUComputeSim
  emitterUpdateScript [
   &EmitterState(Enable Distance Culling=true, MaxDistance=3500, Module.bUseMaxDistance=true, Module.bUseMinDistance=false, Scalability Mode=::NewEnumerator1)
   &SpawnRate(SpawnRate=15, Use Spawn Probability=false)
  ]
  particleSpawnScript [
   &InitializeParticle(
    Lifetime=&UniformRangedFloat(Minimum=1.5, Maximum=3)
    Mass Mode=::NewEnumerator1
    Mesh Scale=&UniformRangedVector(Minimum=<4 4 4>, Maximum=<6 6 6>)
    Mesh Scale Mode=::NewEnumerator4
    Sprite Size Mode=::NewEnumerator4
    UsePositionOffset=false
    Write Material Random=true
   ) {module=InitializeParticle001}
   &InitialMeshOrientation(
    Orientation Vector=/Niagara/DynamicInputs/Random/RandomVector(
     Vector Scale:NiagaraFloat=1
     Constraint Vector:Vector3f=<0 0 1>
     Randomness Mode:ENiagaraRandomnessMode=::NewEnumerator0
     Random Seed:NiagaraInt32=0
     Fixed Random Seed:NiagaraBool=false
    )
    Use Orientation Vector=true
    Use Rotation Vector=false
   ) {module=InitialMeshOrientation001}
   &SphereLocation(
    MaskSpawnGroup=false
    Non Uniform Scale=<1.5 1.5 0.01>
    Offset=<0 0 -130>
    Override Seed=false
    Sphere Coordinate Space=::Simulation
    Sphere Radius=120
    Surface Only=false
    UseNonUniformScale=true
    UseOffset=true
    UseSphereOrientation=false
   ) {module=SphereLocation001}
  ]
  particleUpdateScript [
   &ParticleState()
   &VortexForce(
    Influence Falloff Radius=1500
    Origin Pull Amount=37.9082
    Use Influence Falloff=true
    UseOriginOffset=true
    Vortex Axis=&Lerp_Vector(
     A=<-0.15 0.15 1>
     B=<0.25 0 1>
     Alpha=&FloatFromCurve(
      FloatCurve=DataInterface() {propertyValues {Curve {keys [
          {interpMode=RCIM_Cubic, tangentMode=RCTM_User, time=0, value=0}
          {interpMode=RCIM_Cubic, tangentMode=RCTM_User, time=0.5, value=1}
          {tangentMode=RCTM_User, time=1, value=0}
         ]}}}
      CurveIndex=$System.NormalizedLoopAge
     )
    )
    Vortex Force Amount=1500
    Vortex Origin Offset:Vector3f=<0 0 -25>
   ) {module=VortexForce001}
   /SetVariables(Particles.Position=&Lerp_Vector(
     A=Linked() {linkedVariable {name=Particles.Position, type=NiagaraPosition}}
     B=&Add_Vector(A=Linked() {linkedVariable {name=Particles.Position, type=NiagaraPosition}}, B=<0 0 0>)
     Alpha=$Particles.Lifetime
    )) {isSetParameters=true}
   &DynamicMaterialParameters(
    Index 0 Param 1=&FloatFromCurve(FloatCurve=DataInterface() {propertyValues {Curve {keys [{arriveTangent=-4.0769, time=-0.0004, value=0.9057}, <at 1.0011 0.002>]}}})
    Index 0 Param 2=&FloatFromCurve(FloatCurve=DataInterface() {propertyValues {Curve {keys [<at -0.0046 3.0507>,<at 1.0011 1>]}}}, Scale Curve=3)
    Index 0 Param 3=&FloatFromCurve(
     FloatCurve=DataInterface() {propertyValues {Curve {keys [
         {interpMode=RCIM_Cubic, tangentMode=RCTM_User, time=0, value=30}
         {arriveTangent=22.4428, interpMode=RCIM_Cubic, leaveTangent=22.4428, tangentMode=RCTM_User, time=0.4752, value=45.6649}
         <at 0.5 46.2>
         <at 1 51.4>
         {interpMode=RCIM_Cubic, tangentMode=RCTM_User, time=1.0024, value=51.9752}
         <at 1.2 52>
         <at 1.5 52>
        ]}}}
     Scale Curve=2
    )
    Index 0 Param 4=&FloatFromCurve(
     FloatCurve=DataInterface() {propertyValues {Curve {keys [
         <at 0 0 RCIM_Cubic>
         {arriveTangent=0.002, interpMode=RCIM_Cubic, leaveTangent=0.002, time=0.1245, value=34.51}
         <at 1.0011 0.002 RCIM_Cubic>
        ]}}}
     Scale Curve=0.5
    )
    Module.Param0WriteEnabled=true
    Module.Param1WriteEnabled=true
    Module.Param2WriteEnabled=true
    Module.Param3WriteEnabled=true
   )
   &ScaleColor(
    Scale Alpha=&FloatFromCurve(FloatCurve=DataInterface() {propertyValues {Curve {keys [
         {arriveTangent=2.836, interpMode=RCIM_Cubic, leaveTangent=2.836, tangentMode=RCTM_User, time=0, value=-1}
         {arriveTangent=1.5, interpMode=RCIM_Cubic, leaveTangent=1.5, tangentMode=RCTM_User, time=0.3, value=-0.3}
         <at 0.5 -0.1>
         {arriveTangent=-0.8689, interpMode=RCIM_Cubic, leaveTangent=-0.8689, tangentMode=RCTM_User, time=0.6, value=-0.1}
         <at 1 -1>
         <at 1.0011 -0.9961>
         <at 1.2 -1>
         <at 1.5 -1>
        ]}}})
    ScaleA=true
    ScaleRGB=true
   )
   &ScaleMeshSize(Scale Factor=&VectorFromFloat(Value=&FloatFromCurve(FloatCurve=DataInterface() {propertyValues {Curve {keys [
          {interpMode=RCIM_Cubic, tangentMode=RCTM_User, time=-0.0038, value=0.4623}
          {arriveTangent=1.0062, interpMode=RCIM_Cubic, leaveTangent=1.0062, tangentMode=RCTM_User, time=0.1683, value=0.5805}
          {arriveTangent=-0.1436, interpMode=RCIM_Cubic, leaveTangent=-0.1436, tangentMode=RCTM_User, time=0.543, value=0.7431}
          {arriveTangent=-1.6175, interpMode=RCIM_Cubic, leaveTangent=-1.6175, tangentMode=RCTM_User, time=0.786, value=0.4919}
          <at 1.0011 0.002>
         ]}}})))
   &SolveForcesAndVelocity(Clamp Velocity=true, Limit Acceleration=false, Speed Limit=100)
  ]
  rendererClasses [&NiagaraMeshRendererProperties]
  renderers [{rendererClass=&NiagaraMeshRendererProperties, rendererIndex=0}]
 }
 Fish002 {
  simTarget=GPUComputeSim
  emitterUpdateScript [
   &EmitterState(
    Enable Distance Culling=true
    Inactive Response=::NewEnumerator2
    Life Cycle Mode=::NewEnumerator1
    Loop Behavior=::NewEnumerator1
    Loop Duration Mode=::NewEnumerator1
    MaxDistance=5500
    Min Distance Response=::SleepAndLetParticlesFinish
    MinDistance=250
    Module.bUseMaxDistance=true
    Module.bUseMinDistance=true
    Reset Age On Awaken=true
    Scalability Mode=::NewEnumerator1
    UseLoopDelay=false
   )
   &SpawnRate(SpawnRate=250, Use Spawn Probability=false)
  ]
  particleSpawnScript [
   &InitializeParticle(
    Color Mode=::NewEnumerator0
    Lifetime=3
    Mesh Scale Mode=::NewEnumerator1
    Mesh Uniform Scale Max=3
    Mesh Uniform Scale Min=0.25
    UsePositionOffset=false
    Write Material Random=true
   ) {module=InitializeParticle001}
   &ShapeLocation(
    MaskSpawnGroup=false
    Non Uniform Scale=<1 1 0.25>
    Offset=<0 0 -130>
    Offset Mode=::NewEnumerator0
    Override Seed=false
    Sphere Radius=350
    Sphere Surface Distribution=1
    UseOwnerScale=true
   )
  ]
  particleUpdateScript [
   &ParticleState()
   &InterpolateOverTime(
    Rate Of Change=25
    Target Vector Value:Vector3f=&NormalizeVector(Vector To Normalize=$Particles.RandomDirection, Safe Normalize=true, Safe Fallback Threshold=0, Safe Fallback Vector=<0 0 0.001>)
    Variable Type=::NewEnumerator2
   ) {module=InterpolateOverTime001}
   &PointAttractionForce(
    Attraction Radius=2000
    Attraction Strength=&RandomRangeFloat(Minimum=1, Maximum=5)
    Attractor Position:Vector3f=$Particles.FishMoverPosition
    Attractor Position Offset:Vector3f=<0 0 -150>
    Falloff Exponent=1
    Kill Within Radius=false
    Module.Use Falloff=true
    UseAttractorOffset=true
   )
   &VortexForce(
    Influence Falloff Radius=1500
    Origin Pull Amount=1500
    Use Influence Falloff=true
    UseOriginOffset=false
    Vortex Axis=&Lerp_Vector(
     A=<-0.15 0.15 0>
     B=<0.25 0 0.25>
     Alpha=&FloatFromCurve(
      FloatCurve=DataInterface() {propertyValues {Curve {keys [
          {interpMode=RCIM_Cubic, tangentMode=RCTM_User, time=0, value=0}
          {interpMode=RCIM_Cubic, tangentMode=RCTM_User, time=0.5, value=1}
          {tangentMode=RCTM_User, time=1, value=0}
         ]}}}
      CurveIndex=$System.NormalizedLoopAge
     )
    )
    Vortex Force Amount=5000
    Vortex Origin:NiagaraPosition=&AddVectorToPosition(Position=Linked() {linkedVariable {name=Particles.FishMoverPosition, type=Vector3f}}, Vector=<0 0 -140>)
   ) {module=VortexForce001}
   &ScaleColor(
    Scale Alpha=&FloatFromCurve(FloatCurve=DataInterface() {propertyValues {Curve {keys [
         {arriveTangent=9.0725, interpMode=RCIM_Cubic, leaveTangent=9.0725, tangentMode=RCTM_User, time=0, value=0}
         {arriveTangent=-2.5249, interpMode=RCIM_Cubic, leaveTangent=-2.5249, tangentMode=RCTM_User, time=0.1408, value=0.9847}
         {arriveTangent=-0.0405, interpMode=RCIM_Cubic, leaveTangent=-0.0405, tangentMode=RCTM_User, time=1, value=0}
        ]}}})
    ScaleA=true
    ScaleRGB=true
   )
   &VortexForce(Origin Pull Amount=1000, UseOriginOffset=false, Vortex Force Amount=2500)
   &CurlNoiseForce(Noise Frequency=25, Noise Strength=1000, Pan Noise=false)
   &SolveForcesAndVelocity(
    Clamp Velocity=true
    Limit Acceleration=false
    Speed Limit:NiagaraFloat=&RandomRangeFloat(Minimum=100, Maximum=150)
   )
   &self:ScratchModule_02(CS=15, WH=-20) {module=ScratchModule_02}
  ]
  rendererClasses [&NiagaraMeshRendererProperties]
  renderers [{rendererClass=&NiagaraMeshRendererProperties, rendererIndex=0}]
 }
 Fish003 {
  emitterUpdateScript [
   &EmitterState(
    Enable Distance Culling=true
    Inactive Response=::NewEnumerator2
    Life Cycle Mode=::NewEnumerator1
    Loop Behavior=::NewEnumerator1
    Loop Duration Mode=::NewEnumerator1
    MaxDistance=5500
    Module.bUseMaxDistance=true
    Module.bUseMinDistance=false
    Reset Age On Awaken=true
    Scalability Mode=::NewEnumerator1
    UseLoopDelay=false
   )
   &SpawnBurst_Instantaneous(Spawn Count=10, Use Spawn Probability=false)
   &SpawnBurst_Instantaneous(Spawn Count=4, Spawn Group=1, Use Spawn Probability=false) {module=SpawnBurst_Instantaneous001}
   /SetVariables(Emitter.float:NiagaraFloat=&RandomRangeFloat(Minimum=0, Maximum=100)) {isSetParameters=true}
  ]
  particleSpawnScript [
   &InitializeParticle(
    Color Mode=::NewEnumerator0
    Lifetime=3
    Mesh Scale Mode=::NewEnumerator3
    Mesh Uniform Scale=&MaskFloatBySpawnGroup(Masked Float=&RandomRangeFloat(Minimum=3, Maximum=3.5), Passthrough Float=2.5)
    UsePositionOffset=false
    Write Material Random=true
   ) {module=InitializeParticle001}
   &ShapeLocation(MaskSpawnGroup=false, Non Uniform Scale=<1 1 0.3>, Offset=<0 0 -120>, Offset Mode=::NewEnumerator0, Override Seed=false, UseOwnerScale=true)
   /SetVariables(Particles.LastPos:NiagaraPosition=<0 0 0>) {isSetParameters=true}
   /SetVariables() {isSetParameters=true}
   /SetVariables(Particles.VisibilityTag=$Emitter.SpawnGroup) {isSetParameters=true}
  ]
  particleUpdateScript [
   &ParticleState(Kill Particles When Lifetime Has Elapsed=false, Let Infinitely Lived Particles Die When Emitter Deactivates=true)
   &self:Calc_Tracking_Point(
    Fast=0.5
    Inner Radius=45
    Offset:Vector3f=<0 0 -100>
    Outer Radius=75
    R:NiagaraFloat=&Add_Float(A=&RandomRangeFloat(Minimum=-1, Maximum=1), B=$Emitter.float)
    Radius=400
    ScaleAxis:Vector3f=<1 1 0.2>
    Seed:NiagaraInt32=&RandomRangeInt(Minimum=0, Maximum=100)
   ) {module=Calc_Tracking_Point}
   /SetVariables(Particles.SwimSpeed:NiagaraFloat=&Waveform(
     Global Input Position=&RandomRangeFloat(Minimum=0, Maximum=100)
     "Waveform [1]"=::NewEnumerator12
     "[1] Frequency"=0.5
     "[1] Amplitude Min/Max"=<0.3 1>
    )) {isSetParameters=true}
   &self:SwimToTrackPoint(
    MaxLSHeight=-50
    SwimSpeed:NiagaraFloat=&Multiply_Float(A=$Particles.SwimSpeed, B=350)
    TrackPosition:NiagaraPosition=$Particles.TrackPosition
    TurnSpeed:NiagaraFloat=&Multiply_Float(A=$Particles.SwimSpeed, B=5)
   ) {module=SwimToTrackPoint}
   /SetVariables(Particles.LastPos:NiagaraPosition=$Particles.Position) {isSetParameters=true}
   /SetVariables(Particles.Position=$Particles.SwimToTrackPoint.SwimPosition) {isSetParameters=true}
   &self:Calc_Fish_Direction(
    LastFishDirection:Vector3f=$Particles.LastFishDirection
    LastPosV:NiagaraPosition=$Particles.LastPos
    Max=0.2
    Quat:Quat4f=$Particles.SwimDir
    float001=0.05
   ) {module=Calc_Fish_Direction}
   /SetVariables(Particles.LastFishDirection:Vector3f=$Particles.FishDirection) {isSetParameters=true}
   &InterpolateOverTime(Rate Of Change=8, Target Vector Value:Vector3f=$Particles.TurnAxis, Variable Type=::NewEnumerator2) {module=InterpolateOverTime001}
   &InterpolateOverTime(Rate Of Change=8, Target Float Value:NiagaraFloat=$Particles.TurnAngle) {module=InterpolateOverTime002}
   &Timeline(
    Play Rate=&Multiply_Float(A=$Particles.FishVelocity, B=0.005)
    Rewind Mode:ENiagara_TimelineMode=::NewEnumerator1
    Timeline End=10000000
   )
   &DynamicMaterialParameters(
    Index 0 Param 1=&Add_Float(A=$Particles.Timeline.Time, B=&RandomRangeFloat(Minimum=0, Maximum=50))
    Index 0 Param 2=&MaskFloatBySpawnGroup(
     Masked Float=&SelectFloatFromArray(
      Direct Array Index=&MakeIntFromFloat(
       Float=&FloatFromCurve(
        FloatCurve=DataInterface() {propertyValues {Curve {keys [<at 0 0 RCIM_Constant>,<at 0.45 1 RCIM_Constant>,<at 0.6 2 RCIM_Constant>]}}}
        CurveIndex=&ReturnNormalizedExecIndex()
       )
       Conversion Mode=::NewEnumerator1
      )
      Direct Array Mode=::NewEnumerator1
      Float Selection Array=DataInterface() {propertyValues {FloatData [0,0.4,0.7]}}
     )
     Passthrough Float=&RandomRangeFloat(Minimum=0, Maximum=0)
    )
    Index 0 Param 3=&Add_Float(
     A=&Multiply_Float(A=$Particles.InterpolateOverTime002.InterpolateOverTime_Float.Moving Average, B=10)
     B=&Waveform(
      Global Input Position=&RandomRangeFloat(Minimum=0, Maximum=500)
      "Waveform [1]"=::NewEnumerator12
      "[1] Frequency"=0.4
      "[1] Amplitude Min/Max"=<0.4 0.6>
     )
    )
    Index 0 Param 4=&MaskFloatBySpawnGroup(
     Masked Float=&SelectFloatFromArray(
      Direct Array Index=&MakeIntFromFloat(
       Float=&FloatFromCurve(
        FloatCurve=DataInterface() {propertyValues {Curve {keys [<at 0 0 RCIM_Constant>,<at 0.45 1 RCIM_Constant>,<at 0.6 2 RCIM_Constant>]}}}
        CurveIndex=&ReturnNormalizedExecIndex()
       )
       Conversion Mode=::NewEnumerator1
      )
      Direct Array Mode=::NewEnumerator1
      Float Selection Array=DataInterface() {propertyValues {FloatData [-0.3,0.4,0.4]}}
     )
     Passthrough Float=0.05
    )
    Index 1 Param 1=&MaskFloatBySpawnGroup(Masked Float=&RandomRangeFloat(Minimum=0.4, Maximum=0.6), Passthrough Float=0.8)
    Index 1 Param 2=&FloatFromCurve(
     FloatCurve=DataInterface() {propertyValues {Curve {keys [<at 0 1>,<at 0.4 1>,<at 0.5 0.1>,<at 1 0.1>]}}}
     CurveIndex=&RandomRangeFloat(Minimum=0, Maximum=1)
     Scale Curve=&FNTimeOfDayScalar(Night Scale=0.1, Morning Scale=1, Evening Scale=1)
    )
    Index 3 Param 1=&MakeFloatFromVector(Vector=$Particles.InterpolateOverTime001.InterpolateOverTime_V3_.Moving Average, Channel=::NewEnumerator0)
    Index 3 Param 2=&MakeFloatFromVector(Vector=$Particles.InterpolateOverTime001.InterpolateOverTime_V3_.Moving Average, Channel=::NewEnumerator1)
    Index 3 Param 3=&MakeFloatFromVector(Vector=$Particles.InterpolateOverTime001.InterpolateOverTime_V3_.Moving Average, Channel=::NewEnumerator2)
    Index 3 Param 4=$Particles.InterpolateOverTime002.InterpolateOverTime_Float.Moving Average
    Module.Index0Param0Write=true
    Module.Index0Param1Write=true
    Module.Index0Param2Write=true
    Module.Index0Param3Write=true
    Module.Index1Param0Write=true
    Module.Index1Param1Write=true
    Module.Index1Param2Write=true
    Module.Index1Param3Write=true
    Module.Index3Param0Write=true
    Module.Index3Param1Write=true
    Module.Index3Param2Write=true
    Module.Index3Param3Write=true
    Write Parameter Index 1=true
    Write Parameter Index 3=true
   )
  ]
  rendererClasses [&NiagaraMeshRendererProperties]
  renderers [
   {rendererClass=&NiagaraMeshRendererProperties, rendererIndex=0}
   {rendererClass=&NiagaraMeshRendererProperties, rendererIndex=1}
  ]
 }
}
summary {
 systemName=NS_Flopper_Spot
 emitters [
  {emitterName=Bubbles001, simTarget=GPUComputeSim, rendererClasses [&NiagaraMeshRendererProperties]}
  {emitterName=Fish002, simTarget=GPUComputeSim, rendererClasses [&NiagaraMeshRendererProperties]}
  {emitterName=Fish003, rendererClasses [&NiagaraMeshRendererProperties]}
 ]
 userVariables [User.FishCount {type=NiagaraInt32, defaultValue=3}]
}
