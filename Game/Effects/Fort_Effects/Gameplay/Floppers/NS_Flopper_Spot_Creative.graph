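# Niagara system for the Creative fishing (Flopper) spot: rising bubbles, surface ripple rings and a school of fish below.
# Fish003 bursts two spawn groups of mesh fish that chase moving track points via scratch modules, feeding smoothed turns and velocity-driven time into materials.
# Rings spawn from user parameters, constrain to the water plane, tint by time of day and adjust alpha by distance and platform.
# Older orbit-based fish emitters remain disabled.
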NiagaraSystem /Game/Effects/Fort_Effects/Gameplay/Floppers/NS_Flopper_Spot_Creative
dependencies {
 usedDataInterfaces [
  &NiagaraDataInterfaceCurve
  &NiagaraDataInterfaceColorCurve
  &NiagaraDataInterfacePlatformSet
  &NiagaraDataInterfaceArrayFloat
 ]
 usedDynamicInputs []
 usedModules [/Niagara/Modules/System/SystemLifeCycle, &SystemState]
}
emitters {
 Bubbles {
  emitterUpdateScript [
   &EmitterState(Enable Distance Culling=true, MaxDistance=3500, Module.bUseMaxDistance=true, Module.bUseMinDistance=false, Scalability Mode=::NewEnumerator1)
   &SpawnRate(SpawnRate=5, Use Spawn Probability=false)
  ]
  particleSpawnScript [
   /Game/Effects/Niagara/Modules/Orbit/FN_Orbit_Init_Variables_v2()
   /Niagara/Modules/Spawn/Initialization/InitializeParticle(
    Lifetime=&UniformRangedFloat(Minimum=1.5, Maximum=3)
    Mesh Scale=&UniformRangedVector(Minimum=<2 2 1>, Maximum=<5 5 1>)
    Module.Write Color=true
    Module.Write Lifetime=true
    Module.Write Mass=true
    Module.Write Material Random=true
    Module.Write Position=false
    Module.Write Scale=true
    Module.Write SpriteRotation=false
    Module.Write SpriteSize=false
    Position=&SimulationPosition()
   )
   /Niagara/Modules/Spawn/Orientation/V2/InitialMeshOrientation(
    Orientation Vector=/Niagara/DynamicInputs/Random/RandomVector(
     Vector Scale:NiagaraFloat=1
     Constraint Vector:Vector3f=<0 0 1>
     Randomness Mode:ENiagaraRandomnessMode=::NewEnumerator0
     Random Seed:NiagaraInt32=0
     Fixed Random Seed:NiagaraBool=false
    )
    Use Orientation Vector=true
    Use Rotation Vector=false
   ) {enabled=false}
   /Niagara/Modules/Spawn/Orientation/V3/InitialMeshOrientation(
    Orientation Vector=&RandomVector(
     Vector Scale:NiagaraFloat=1
     Constraint Vector:Vector3f=<0 0 1>
     Randomness Mode:ENiagaraRandomnessMode=::NewEnumerator0
     Random Seed:NiagaraInt32=0
     Fixed Random Seed:NiagaraBool=false
    )
    Use Orientation Vector=true
    Use Rotation Vector=false
   ) {module=InitialMeshOrientation001}
   &SphereLocation(
    MaskSpawnGroup=false
    Non Uniform Scale=<1.5 1.5 0.01>
    Offset=<0 0 -200>
    Override Seed=false
    Sphere Coordinate Space=::Simulation
    Sphere Radius=150
    Surface Only=false
    UseNonUniformScale=true
    UseOffset=true
    UseSphereOrientation=false
   ) {module=SphereLocation001}
  ]
  particleUpdateScript [
   /SetVariables(Particles.ScaleCorrection:NiagaraFloat=&ClampFloat(Float=&MakeFloatFromVector(Vector=$Engine.Owner.Scale, Channel=::NewEnumerator0), Min=0.3, Max=3)) {isSetParameters=true}
   &ParticleState()
   &GravityForce(Gravity=<0 0 100>) {enabled=false}
   /SetVariables(Particles.Position=&Lerp_Vector(
     A=Linked() {linkedVariable {name=Particles.Position, type=NiagaraPosition}}
     B=&Add_Vector(A=Linked() {linkedVariable {name=Particles.Position, type=NiagaraPosition}}, B=<0 0 0.4>)
     Alpha=$Particles.Lifetime
    )) {isSetParameters=true}
   &DynamicMaterialParameters(
    Index 0 Param 1=&FloatFromCurve(FloatCurve=DataInterface() {propertyValues {Curve {keys [{arriveTangent=-4.0769, time=-0.0004, value=0.9057}, <at 1.0011 0.002>]}}})
    Index 0 Param 2=&FloatFromCurve(FloatCurve=DataInterface() {propertyValues {Curve {keys [<at -0.0046 3.0507>,<at 1.0011 1>]}}})
    Index 0 Param 3=&FloatFromCurve(
     FloatCurve=DataInterface() {propertyValues {Curve {keys [
         {interpMode=RCIM_Cubic, tangentMode=RCTM_User, time=0, value=30}
         {arriveTangent=22.4428, interpMode=RCIM_Cubic, leaveTangent=22.4428, tangentMode=RCTM_User, time=0.4752, value=45.6649}
         <at 0.5 46.2>
         <at 1 51.4>
         {interpMode=RCIM_Cubic, tangentMode=RCTM_User, time=1.0024, value=51.9752}
         <at 1.2 52>
         <at 1.5 52>
        ]}}}
     Scale Curve=$Particles.ScaleCorrection
    )
    Index 0 Param 4=&FloatFromCurve(
     FloatCurve=DataInterface() {propertyValues {Curve {keys [
         <at 0 0 RCIM_Cubic>
         {arriveTangent=0.002, interpMode=RCIM_Cubic, leaveTangent=0.002, time=0.1245, value=34.51}
         <at 1.0011 0.002 RCIM_Cubic>
        ]}}}
     Scale Curve=$Particles.ScaleCorrection
    )
    Module.Param0WriteEnabled=true
    Module.Param1WriteEnabled=true
    Module.Param2WriteEnabled=true
    Module.Param3WriteEnabled=true
   )
   &ScaleColor(
    Scale Alpha=&FloatFromCurve(FloatCurve=DataInterface() {propertyValues {Curve {keys [
         {arriveTangent=2.836, interpMode=RCIM_Cubic, leaveTangent=2.836, tangentMode=RCTM_User, time=0, value=-1}
         {arriveTangent=1.5, interpMode=RCIM_Cubic, leaveTangent=1.5, tangentMode=RCTM_User, time=0.3, value=-0.3}
         <at 0.5 -0.1>
         {arriveTangent=-0.8689, interpMode=RCIM_Cubic, leaveTangent=-0.8689, tangentMode=RCTM_User, time=0.6, value=-0.1}
         <at 1 -1>
         <at 1.0011 -0.9961>
         <at 1.2 -1>
         <at 1.5 -1>
        ]}}})
    ScaleA=true
    ScaleRGB=true
   )
   &ScaleMeshSize(Scale Factor=&VectorFromFloat(Value=&FloatFromCurve(FloatCurve=DataInterface() {propertyValues {Curve {keys [
          {interpMode=RCIM_Cubic, tangentMode=RCTM_User, time=-0.0038, value=0.4623}
          {arriveTangent=1.0062, interpMode=RCIM_Cubic, leaveTangent=1.0062, tangentMode=RCTM_User, time=0.1683, value=0.5805}
          {arriveTangent=-0.1436, interpMode=RCIM_Cubic, leaveTangent=-0.1436, tangentMode=RCTM_User, time=0.543, value=0.7431}
          {arriveTangent=-1.6175, interpMode=RCIM_Cubic, leaveTangent=-1.6175, tangentMode=RCTM_User, time=0.786, value=0.4919}
          <at 1.0011 0.002>
         ]}}}))) {enabled=false}
   /Niagara/Modules/Update/Orientation/V2/UpdateMeshOrientation(
    Rotation Rate:NiagaraFloat=&UniformRangedFloat(Minimum=0.1, Maximum=0.25)
    Rotation Vector:Vector3f=&UniformRangedVector(Minimum=<0 0 0.1>, Maximum=<0 0 0.25>)
   ) {enabled=false}
  ]
  rendererClasses [&NiagaraMeshRendererProperties]
  renderers [{rendererClass=&NiagaraMeshRendererProperties, rendererIndex=0}]
 }
 Fish {
  bEnabled=false
  emitterUpdateScript [
   &EmitterState(
    Enable Distance Culling=true
    Life Cycle Mode=::NewEnumerator1
    Loop Behavior=::NewEnumerator1
    Loop Duration Mode=::NewEnumerator1
    MaxDistance=10000
    Module.bUseMaxDistance=true
    Module.bUseMinDistance=false
    Scalability Mode=::NewEnumerator1
    UseLoopDelay=false
   )
   &SpawnBurst_Instantaneous(Spawn Count=$User.FishCount, Use Spawn Probability=false)
  ]
  particleSpawnScript [
   &FN_Orbit_Init_Variables_v2()
   &InitializeParticle(
    Mesh Scale=<12 12 12>
    Module.Write Color=true
    Module.Write Lifetime=true
    Module.Write Mass=true
    Module.Write Material Random=true
    Module.Write Position=false
    Module.Write Scale=true
    Module.Write SpriteRotation=false
    Module.Write SpriteSize=false
    Position=&SimulationPosition()
   )
   &SystemLocation(MaskSpawnGroup=false, Offset=&UniformRangedVector(Minimum=<-25 0 -165>, Maximum=<25 1 -185>))
   /Niagara/Modules/Spawn/Location/SphereLocation(MaskSpawnGroup=false, Module.Override Randomness=false, Non Uniform Scale=<1 1 0.2>, Sphere Radius=25, Surface Only=false)
  ]
  particleUpdateScript [
   /Game/Effects/Niagara/Modules/Orbit/FN_Orbit_v3(
    Height min/max:Vector2f=<0 150>
    Radius min/max:Vector2f=<125 150>
    Rotation Offset min/max:Vector2f=<1 5>
    Rotation Rate min/max:Vector2f=<0.1 0.2>
    To Axis:Vector3f=&Lerp_Vector(A=<-0.25 0 1>, B=<0.25 0.25 1>, Alpha=0.5)
   ) {enabled=false}
   /Game/Effects/Niagara/Modules/Orbit/FN_Orbit_v4(
    Height min/max:Vector2f=<0 150>
    Radius min/max:Vector2f=<125 150>
    Rotation Offset min/max:Vector2f=<1 5>
    Rotation Rate min/max:Vector2f=<0.1 0.2>
    To Axis:Vector3f=&Lerp_Vector(A=<-0.25 0 1>, B=<0.25 0.25 1>, Alpha=0.5)
   )
   &SolveForcesAndVelocity(Clamp Velocity=false, Limit Acceleration=false)
   &ParticleState(Kill Particles When Lifetime Has Elapsed=false, Let Infinitely Lived Particles Die When Emitter Deactivates=true)
   &Color()
  ]
  rendererClasses [&NiagaraMeshRendererProperties]
  renderers [{rendererClass=&NiagaraMeshRendererProperties, rendererIndex=0}]
 }
 Fish001 {
  bEnabled=false
  emitterUpdateScript [
   &EmitterState(
    Enable Distance Culling=true
    Inactive Response=::NewEnumerator2
    Life Cycle Mode=::NewEnumerator1
    Loop Behavior=::NewEnumerator1
    Loop Duration Mode=::NewEnumerator1
    MaxDistance=5500
    Module.bUseMaxDistance=true
    Module.bUseMinDistance=false
    Reset Age On Awaken=true
    Scalability Mode=::NewEnumerator1
    UseLoopDelay=false
   )
   &SpawnBurst_Instantaneous(Spawn Count=&Multiply_Int(A=$User.FishCount, B=3), Use Spawn Probability=false)
  ]
  particleSpawnScript [
   /SetVariables(Particles.MeshIndex=&RandomRangeInt(Minimum=0, Maximum=2, Randomness Mode=::NewEnumerator2, Random Seed=?)) {isSetParameters=true}
   &FN_Orbit_Init_Variables_v2()
   /Niagara/Modules/Spawn/Initialization/V2/InitializeParticle(
    Color Mode=::NewEnumerator0
    Mesh Scale Mode=::NewEnumerator3
    Mesh Uniform Scale=&MaskFloatBySpawnGroup(Masked Float=12, Passthrough Float=10)
    UsePositionOffset=false
    Write Material Random=true
   ) {module=InitializeParticle001}
   &SystemLocation(MaskSpawnGroup=false, Offset=&UniformRangedVector(Minimum=<0 0 -210>, Maximum=<1 1 -195>))
   /Niagara/Modules/Spawn/Location/V2/SphereLocation(
    MaskSpawnGroup=false
    Non Uniform Scale=<1 1 0.2>
    Override Seed=false
    Sphere Radius=25
    Surface Only=false
    UseNonUniformScale=true
    UseOffset=false
    UseSphereOrientation=false
   ) {module=SphereLocation001}
  ]
  particleUpdateScript [
   /SetVariables(Particles.AgeTest:NiagaraFloat=$Emitter.Age) {isSetParameters=true}
   &FN_Orbit_v4(
    Height min/max:Vector2f=<0 150>
    Radius min/max:Vector2f=<125 150>
    Rotation Offset min/max:Vector2f=<1 5>
    Rotation Rate min/max:Vector2f=<0.1 0.2>
    To Axis:Vector3f=&Lerp_Vector(A=<-0.25 0 1>, B=<0.25 0.25 1>, Alpha=0.5)
   )
   &ParticleState(Kill Particles When Lifetime Has Elapsed=false, Let Infinitely Lived Particles Die When Emitter Deactivates=true)
  ]
  rendererClasses [&NiagaraMeshRendererProperties]
  renderers [{rendererClass=&NiagaraMeshRendererProperties, rendererIndex=0}]
 }
 Fish003 {
  emitterUpdateScript [
   &EmitterState(
    Enable Distance Culling=true
    Inactive Response=::NewEnumerator2
    Life Cycle Mode=::NewEnumerator1
    Loop Behavior=::NewEnumerator1
    Loop Duration Mode=::NewEnumerator1
    MaxDistance=5500
    Module.bUseMaxDistance=true
    Module.bUseMinDistance=false
    Reset Age On Awaken=true
    Scalability Mode=::NewEnumerator1
    UseLoopDelay=false
   )
   &SpawnBurst_Instantaneous(Spawn Count=10, Use Spawn Probability=false)
   &SpawnBurst_Instantaneous(Spawn Count=4, Spawn Group=1, Use Spawn Probability=false) {module=SpawnBurst_Instantaneous001}
   /SetVariables(Emitter.float:NiagaraFloat=&RandomRangeFloat(Minimum=0, Maximum=100)) {isSetParameters=true}
  ]
  particleSpawnScript [
   &InitializeParticle(
    Color Mode=::NewEnumerator0
    Lifetime=3
    Mesh Scale Mode=::NewEnumerator3
    Mesh Uniform Scale=&MaskFloatBySpawnGroup(Masked Float=&RandomRangeFloat(Minimum=3, Maximum=3.5), Passthrough Float=2.5)
    UsePositionOffset=false
    Write Material Random=true
   ) {module=InitializeParticle001}
   &ShapeLocation(MaskSpawnGroup=false, Non Uniform Scale=<1 1 0.3>, Offset=<0 0 -120>, Offset Mode=::NewEnumerator0, Override Seed=false, UseOwnerScale=true)
   /SetVariables(Particles.LastPos:NiagaraPosition=<0 0 0>) {isSetParameters=true}
   /SetVariables() {isSetParameters=true}
   /SetVariables(Particles.VisibilityTag=$Emitter.SpawnGroup) {isSetParameters=true}
  ]
  particleUpdateScript [
   &ParticleState(Kill Particles When Lifetime Has Elapsed=false, Let Infinitely Lived Particles Die When Emitter Deactivates=true)
   &self:ScratchModule(
    Fast=0.5
    Inner Radius=45
    Offset:Vector3f=<0 0 -100>
    Outer Radius=75
    R:NiagaraFloat=&Add_Float(A=&RandomRangeFloat(Minimum=-1, Maximum=1), B=$Emitter.float)
    Radius=400
    ScaleAxis:Vector3f=<1 1 0.2>
    Seed:NiagaraInt32=&RandomRangeInt(Minimum=0, Maximum=100)
   ) {module=ScratchModule}
   /SetVariables(Particles.SwimSpeed:NiagaraFloat=&Waveform(
     Global Input Position=&RandomRangeFloat(Minimum=0, Maximum=100)
     "Waveform [1]"=::NewEnumerator12
     "[1] Frequency"=0.5
     "[1] Amplitude Min/Max"=<0.3 1>
    )) {isSetParameters=true}
   &self:SwimToTrackPoint(
    MaxLSHeight=-50
    SwimSpeed:NiagaraFloat=&Multiply_Float(A=$Particles.SwimSpeed, B=350)
    TrackPosition:NiagaraPosition=$Particles.TrackPosition
    TurnSpeed:NiagaraFloat=&Multiply_Float(A=$Particles.SwimSpeed, B=5)
   ) {module=SwimToTrackPoint}
   /SetVariables(Particles.LastPos:NiagaraPosition=$Particles.Position) {isSetParameters=true}
   /SetVariables(Particles.Position=$Particles.SwimToTrackPoint.SwimPosition) {isSetParameters=true}
   &self:ScratchModule_01(
    LastFishDirection:Vector3f=$Particles.LastFishDirection
    LastPosV:NiagaraPosition=$Particles.LastPos
    Max=0.2
    Quat:Quat4f=$Particles.SwimDir
    float001=0.05
   ) {module=ScratchModule_01}
   /SetVariables(Particles.LastFishDirection:Vector3f=$Particles.FishDirection) {isSetParameters=true}
   &InterpolateOverTime(Rate Of Change=8, Target Vector Value:Vector3f=$Particles.TurnAxis, Variable Type=::NewEnumerator2) {module=InterpolateOverTime001}
   &InterpolateOverTime(Rate Of Change=8, Target Float Value:NiagaraFloat=$Particles.TurnAngle) {module=InterpolateOverTime002}
   &Timeline(
    Play Rate=&Multiply_Float(A=$Particles.FishVelocity, B=0.005)
    Rewind Mode:ENiagara_TimelineMode=::NewEnumerator1
    Timeline End=10000000
   )
   &DynamicMaterialParameters(
    Index 0 Param 1=&Add_Float(A=$Particles.Timeline.Time, B=&RandomRangeFloat(Minimum=0, Maximum=50))
    Index 0 Param 2=&MaskFloatBySpawnGroup(
     Masked Float=&SelectFloatFromArray(
      Direct Array Index=&MakeIntFromFloat(
       Float=&FloatFromCurve(
        FloatCurve=DataInterface() {propertyValues {Curve {keys [<at 0 0 RCIM_Constant>,<at 0.45 1 RCIM_Constant>,<at 0.6 2 RCIM_Constant>]}}}
        CurveIndex=&ReturnNormalizedExecIndex()
       )
       Conversion Mode=::NewEnumerator1
      )
      Direct Array Mode=::NewEnumerator1
      Float Selection Array=DataInterface() {propertyValues {FloatData [0,0.4,0.7]}}
     )
     Passthrough Float=&RandomRangeFloat(Minimum=0, Maximum=0)
    )
    Index 0 Param 3=&Add_Float(
     A=&Multiply_Float(A=$Particles.InterpolateOverTime002.InterpolateOverTime_Float.Moving Average, B=10)
     B=&Waveform(
      Global Input Position=&RandomRangeFloat(Minimum=0, Maximum=500)
      "Waveform [1]"=::NewEnumerator12
      "[1] Frequency"=0.4
      "[1] Amplitude Min/Max"=<0.4 0.6>
     )
    )
    Index 0 Param 4=&MaskFloatBySpawnGroup(
     Masked Float=&SelectFloatFromArray(
      Direct Array Index=&MakeIntFromFloat(
       Float=&FloatFromCurve(
        FloatCurve=DataInterface() {propertyValues {Curve {keys [<at 0 0 RCIM_Constant>,<at 0.45 1 RCIM_Constant>,<at 0.6 2 RCIM_Constant>]}}}
        CurveIndex=&ReturnNormalizedExecIndex()
       )
       Conversion Mode=::NewEnumerator1
      )
      Direct Array Mode=::NewEnumerator1
      Float Selection Array=DataInterface() {propertyValues {FloatData [-0.3,0.4,0.4]}}
     )
     Passthrough Float=0.05
    )
    Index 1 Param 1=&MaskFloatBySpawnGroup(Masked Float=&RandomRangeFloat(Minimum=0.4, Maximum=0.6), Passthrough Float=0.8)
    Index 1 Param 2=&FloatFromCurve(
     FloatCurve=DataInterface() {propertyValues {Curve {keys [<at 0 1>,<at 0.4 1>,<at 0.5 0.1>,<at 1 0.1>]}}}
     CurveIndex=&RandomRangeFloat(Minimum=0, Maximum=1)
     Scale Curve=&FNTimeOfDayScalar(Night Scale=0.1, Morning Scale=1, Evening Scale=1)
    )
    Index 3 Param 1=&MakeFloatFromVector(Vector=$Particles.InterpolateOverTime001.InterpolateOverTime_V3_.Moving Average, Channel=::NewEnumerator0)
    Index 3 Param 2=&MakeFloatFromVector(Vector=$Particles.InterpolateOverTime001.InterpolateOverTime_V3_.Moving Average, Channel=::NewEnumerator1)
    Index 3 Param 3=&MakeFloatFromVector(Vector=$Particles.InterpolateOverTime001.InterpolateOverTime_V3_.Moving Average, Channel=::NewEnumerator2)
    Index 3 Param 4=$Particles.InterpolateOverTime002.InterpolateOverTime_Float.Moving Average
    Module.Index0Param0Write=true
    Module.Index0Param1Write=true
    Module.Index0Param2Write=true
    Module.Index0Param3Write=true
    Module.Index1Param0Write=true
    Module.Index1Param1Write=true
    Module.Index1Param2Write=true
    Module.Index1Param3Write=true
    Module.Index3Param0Write=true
    Module.Index3Param1Write=true
    Module.Index3Param2Write=true
    Module.Index3Param3Write=true
    Write Parameter Index 1=true
    Write Parameter Index 3=true
   )
   /SetVariables(Particles.Position=$Particles.TrackPosition) {enabled=false, isSetParameters=true}
  ]
  rendererClasses [&NiagaraMeshRendererProperties]
  renderers [
   {rendererClass=&NiagaraMeshRendererProperties, rendererIndex=0}
   {rendererClass=&NiagaraMeshRendererProperties, rendererIndex=1}
  ]
 }
 Fish_AllInOne {
  bEnabled=false
  emitterUpdateScript [
   &EmitterScalability(
    Module.bApplySpawnCountScale=false
    Module.bApplySpawnCountScaleByDistanceFraction=false
    Module.bUseMaxDistance=false
    Module.bUseMinDistance=false
   ) {enabled=false}
   &EmitterLifeCycle() {enabled=false}
   &EmitterState(Life Cycle Mode=::NewEnumerator1, Loop Behavior=::NewEnumerator1, Loop Duration Mode=::NewEnumerator1, UseLoopDelay=false)
   &SpawnBurst_Instantaneous(Spawn Count=$User.FishCount, Use Spawn Probability=false)
  ]
  particleSpawnScript [
   &FN_Orbit_Init_Variables_v2()
   /Niagara/Modules/Spawn/Initialization/InitializeParticle(
    Mesh Scale=<10 10 10>
    Module.Write Color=true
    Module.Write Lifetime=true
    Module.Write Mass=true
    Module.Write Material Random=true
    Module.Write Position=false
    Module.Write Scale=true
    Module.Write SpriteRotation=false
    Module.Write SpriteSize=false
    Position=&SimulationPosition()
   )
   &AddRotationalVelocity(
    Coordinate Space:ENiagaraMeshTransforms=::NewEnumerator1
    Rotation Rate=&UniformRangedVector(Minimum=<-0.25 0 0>, Maximum=<0.25 0 0>)
   ) {enabled=false}
   &ApplyInitialForces() {enabled=false}
   /Niagara/Modules/Spawn/Orientation/V2/InitialMeshOrientation(Use Orientation Vector=true, Use Rotation Vector=true) {enabled=false}
   &SystemLocation(MaskSpawnGroup=false, Offset=&UniformRangedVector(Minimum=<0 0 -180>, Maximum=<1 0 -220>))
   /Niagara/Modules/Spawn/Location/SphereLocation(MaskSpawnGroup=false, Module.Override Randomness=false, Non Uniform Scale=<1 1 0.2>, Sphere Radius=25, Surface Only=false)
   /SetVariables(Particles.VisibilityTag=&UniformRangedInt(Minimum=0, Maximum=5)) {isSetParameters=true, module=SetVariables_#1}
  ]
  particleUpdateScript [
   &FN_Orbit_v3(
    Height min/max:Vector2f=<0 150>
    Radius min/max:Vector2f=<200 300>
    Rotation Offset min/max:Vector2f=<1 5>
    Rotation Rate min/max:Vector2f=<0.1 0.2>
   )
   &SolveForcesAndVelocity(Clamp Velocity=false, Limit Acceleration=false)
   &ParticleState(Kill Particles When Lifetime Has Elapsed=false, Let Infinitely Lived Particles Die When Emitter Deactivates=true)
   &Color()
  ]
  rendererClasses [&NiagaraMeshRendererProperties]
  renderers [
   {rendererClass=&NiagaraMeshRendererProperties, rendererIndex=0}
   {rendererClass=&NiagaraMeshRendererProperties, rendererIndex=1}
   {rendererClass=&NiagaraMeshRendererProperties, rendererIndex=2}
   {rendererClass=&NiagaraMeshRendererProperties, rendererIndex=3}
   {rendererClass=&NiagaraMeshRendererProperties, rendererIndex=4}
   {rendererClass=&NiagaraMeshRendererProperties, rendererIndex=5}
  ]
 }
 Fish_Effective {
  bEnabled=false
  emitterUpdateScript [
   &EmitterState(
    Enable Distance Culling=true
    Life Cycle Mode=::NewEnumerator1
    Loop Behavior=::NewEnumerator1
    Loop Duration Mode=::NewEnumerator1
    MaxDistance=10000
    Module.bUseMaxDistance=true
    Module.bUseMinDistance=false
    Scalability Mode=::NewEnumerator1
    UseLoopDelay=false
   )
   &SpawnBurst_Instantaneous(Spawn Count=$User.FishCount, Use Spawn Probability=false)
  ]
  particleSpawnScript [
   &FN_Orbit_Init_Variables_v2()
   &InitializeParticle(
    Mesh Scale=<10 10 10>
    Module.Write Color=true
    Module.Write Lifetime=true
    Module.Write Mass=true
    Module.Write Material Random=true
    Module.Write Position=false
    Module.Write Scale=true
    Module.Write SpriteRotation=false
    Module.Write SpriteSize=false
    Position=&SimulationPosition()
   )
   &AddRotationalVelocity(
    Coordinate Space:ENiagaraMeshTransforms=::NewEnumerator1
    Rotation Rate=&UniformRangedVector(Minimum=<-0.25 0 0>, Maximum=<0.25 0 0>)
   ) {enabled=false}
   &ApplyInitialForces()
   &InitialMeshOrientation(Use Orientation Vector=true, Use Rotation Vector=true) {enabled=false}
   &SystemLocation(MaskSpawnGroup=false, Offset=&UniformRangedVector(Minimum=<0 0 -165>, Maximum=<1 0 -180>))
   &SphereLocation(MaskSpawnGroup=false, Module.Override Randomness=false, Non Uniform Scale=<1 1 0.2>, Sphere Radius=25, Surface Only=false)
  ]
  particleUpdateScript [
   &FN_Orbit_v3(
    Height min/max:Vector2f=<0 150>
    Radius min/max:Vector2f=<165 200>
    Rotation Offset min/max:Vector2f=<1 5>
    Rotation Rate min/max:Vector2f=<0.1 0.2>
    To Axis:Vector3f=&Lerp_Vector(A=<0 0 1>, B=<0 0 1>, Alpha=0.5)
   ) {enabled=false}
   &FN_Orbit_v4(
    Height min/max:Vector2f=<0 150>
    Radius min/max:Vector2f=<165 200>
    Rotation Offset min/max:Vector2f=<1 5>
    Rotation Rate min/max:Vector2f=<0.1 0.2>
   )
   &SolveForcesAndVelocity(Clamp Velocity=false, Limit Acceleration=false)
   &ParticleState(Kill Particles When Lifetime Has Elapsed=false, Let Infinitely Lived Particles Die When Emitter Deactivates=true)
   &Color()
  ]
  rendererClasses [&NiagaraMeshRendererProperties]
  renderers [{rendererClass=&NiagaraMeshRendererProperties, rendererIndex=0}]
 }
 Fish_Less_Effective {
  bEnabled=false
  emitterUpdateScript [
   &EmitterState(
    Enable Distance Culling=true
    Life Cycle Mode=::NewEnumerator1
    Loop Behavior=::NewEnumerator1
    Loop Duration Mode=::NewEnumerator1
    MaxDistance=10000
    Module.bUseMaxDistance=true
    Module.bUseMinDistance=false
    Scalability Mode=::NewEnumerator1
    UseLoopDelay=false
   )
   &SpawnBurst_Instantaneous(Spawn Count=$User.FishCount, Use Spawn Probability=false)
  ]
  particleSpawnScript [
   &FN_Orbit_Init_Variables_v2()
   &InitializeParticle(
    Mesh Scale=<10 10 10>
    Module.Write Color=true
    Module.Write Lifetime=true
    Module.Write Mass=true
    Module.Write Material Random=true
    Module.Write Position=false
    Module.Write Scale=true
    Module.Write SpriteRotation=false
    Module.Write SpriteSize=false
    Position=&SimulationPosition()
   )
   &AddRotationalVelocity(
    Coordinate Space:ENiagaraMeshTransforms=::NewEnumerator1
    Rotation Rate=&UniformRangedVector(Minimum=<-0.25 0 0>, Maximum=<0.25 0 0>)
   ) {enabled=false}
   &ApplyInitialForces() {enabled=false}
   &InitialMeshOrientation(Use Orientation Vector=true, Use Rotation Vector=true) {enabled=false}
   &SystemLocation(MaskSpawnGroup=false, Offset=&UniformRangedVector(Minimum=<0 0 -180>, Maximum=<1 0 -220>))
   &SphereLocation(MaskSpawnGroup=false, Module.Override Randomness=false, Non Uniform Scale=<1 1 0.2>, Sphere Radius=25, Surface Only=false)
  ]
  particleUpdateScript [
   &FN_Orbit_v3(
    Height min/max:Vector2f=<0 150>
    Radius min/max:Vector2f=<200 300>
    Rotation Offset min/max:Vector2f=<1 5>
    Rotation Rate min/max:Vector2f=<0.1 0.2>
   ) {enabled=false}
   &FN_Orbit_v4(
    Height min/max:Vector2f=<0 150>
    Radius min/max:Vector2f=<200 300>
    Rotation Offset min/max:Vector2f=<1 5>
    Rotation Rate min/max:Vector2f=<0.1 0.2>
   )
   &SolveForcesAndVelocity(Clamp Velocity=false, Limit Acceleration=false)
   &ParticleState(Kill Particles When Lifetime Has Elapsed=false, Let Infinitely Lived Particles Die When Emitter Deactivates=true)
   &Color()
  ]
  rendererClasses [&NiagaraMeshRendererProperties]
  renderers [{rendererClass=&NiagaraMeshRendererProperties, rendererIndex=0}]
 }
 Fish_RareFishes {
  bEnabled=false
  emitterUpdateScript [
   &EmitterState(
    Enable Distance Culling=true
    Life Cycle Mode=::NewEnumerator1
    Loop Behavior=::NewEnumerator1
    Loop Duration Mode=::NewEnumerator1
    MaxDistance=10000
    Module.bUseMaxDistance=true
    Module.bUseMinDistance=false
    Scalability Mode=::NewEnumerator1
    UseLoopDelay=false
   )
   &SpawnBurst_Instantaneous(Spawn Count=1, Use Spawn Probability=false)
  ]
  particleSpawnScript [
   &FN_Orbit_Init_Variables_v2()
   &InitializeParticle(
    Mesh Scale=<10 10 10>
    Module.Write Color=true
    Module.Write Lifetime=true
    Module.Write Mass=true
    Module.Write Material Random=true
    Module.Write Position=false
    Module.Write Scale=true
    Module.Write SpriteRotation=false
    Module.Write SpriteSize=false
    Position=&SimulationPosition()
   )
   &AddRotationalVelocity(
    Coordinate Space:ENiagaraMeshTransforms=::NewEnumerator1
    Rotation Rate=&UniformRangedVector(Minimum=<-0.25 0 0>, Maximum=<0.25 0 0>)
   ) {enabled=false}
   &ApplyInitialForces() {enabled=false}
   &InitialMeshOrientation(Use Orientation Vector=true, Use Rotation Vector=true) {enabled=false}
   &SystemLocation(MaskSpawnGroup=false, Offset=&UniformRangedVector(Minimum=<0 0 -180>, Maximum=<1 0 -220>))
   &SphereLocation(MaskSpawnGroup=false, Module.Override Randomness=false, Non Uniform Scale=<1 1 0.2>, Sphere Radius=25, Surface Only=false)
   /SetVariables(Particles.VisibilityTag=&UniformRangedInt(Minimum=0, Maximum=0)) {isSetParameters=true, module=SetVariables_#1}
  ]
  particleUpdateScript [
   &FN_Orbit_v3(
    Height min/max:Vector2f=<0 150>
    Radius min/max:Vector2f=<200 300>
    Rotation Offset min/max:Vector2f=<1 5>
    Rotation Rate min/max:Vector2f=<0.1 0.2>
   ) {enabled=false}
   &FN_Orbit_v4(
    Height min/max:Vector2f=<0 150>
    Radius min/max:Vector2f=<200 300>
    Rotation Offset min/max:Vector2f=<1 5>
    Rotation Rate min/max:Vector2f=<0.1 0.2>
   )
   &SolveForcesAndVelocity(Clamp Velocity=false, Limit Acceleration=false)
   &ParticleState(Kill Particles When Lifetime Has Elapsed=false, Let Infinitely Lived Particles Die When Emitter Deactivates=true)
   &Color()
  ]
  rendererClasses [&NiagaraMeshRendererProperties]
  renderers [
   {rendererClass=&NiagaraMeshRendererProperties, rendererIndex=0}
   {rendererClass=&NiagaraMeshRendererProperties, rendererIndex=1}
   {rendererClass=&NiagaraMeshRendererProperties, rendererIndex=2}
  ]
 }
 Rings {
  emitterUpdateScript [
   &EmitterState(
    Enable Distance Culling=true
    MaxDistance=30000
    Module.bUseMaxDistance=true
    Module.bUseMinDistance=false
    Scalability Mode=::NewEnumerator1
   )
   &SpawnBurst_Instantaneous(Spawn Count=1, Use Spawn Probability=false)
   &SpawnRate(Spawn Probability=0.5, SpawnRate=$User.RippleSpawnRate, Use Spawn Probability=true)
  ]
  particleSpawnScript [
   &InitializeParticle(
    Color=<rgba 1 1 1 0.15>
    Lifetime=&UniformRangedFloat(Minimum=4, Maximum=5)
    Mass=&UniformRangedFloat(Minimum=0.6, Maximum=1.2)
    Mesh Scale=&Multiply_VectorByFloat(Vector=<500 500 500>, Float=$User.RippleScaling)
    Module.Write Color=true
    Module.Write Lifetime=true
    Module.Write Mass=true
    Module.Write Material Random=true
    Module.Write Position=false
    Module.Write Scale=true
    Module.Write SpriteRotation=true
    Module.Write SpriteSize=false
    Position=&SimulationPosition()
    Sprite Rotation=&NormalizedAngleToDegrees(NormalizedAngle=&UniformRangedFloat(Minimum=0, Maximum=1))
   )
   /Niagara/Modules/Spawn/Orientation/V3/InitialMeshOrientation(
    Orientation Vector=<0 0 0>
    Rotation=&UniformRangedVector(Minimum=<0 0 0>, Maximum=<0 0 1>)
    Use Orientation Vector=true
    Use Rotation Vector=true
   ) {module=InitialMeshOrientation001}
   &DynamicMaterialParameters(
    Index 1 Param 3=200
    Index 3 Param 1=$User.WaterBodyIndex
    Index 3 Param 2=$User.WaterDepth
    Module.Param0WriteEnabled1=false
    Module.Param0WriteEnabled3=true
    Module.Param1WriteEnabled1=false
    Module.Param1WriteEnabled3=true
    Module.Param2WriteEnabled1=true
    Module.Param2WriteEnabled3=true
    Module.Param3WriteEnabled1=false
    Module.Param3WriteEnabled3=true
    Write Parameter Index 0=false
    Write Parameter Index 1=true
    Write Parameter Index 3=true
   )
   &InitializeWaterSnapToOwner()
   &SystemLocation(MaskSpawnGroup=false, Offset=<0 0 5>)
  ]
  particleUpdateScript [
   &ParticleState()
   &DynamicMaterialParameters(
    Index 0 Param 1=&FloatFromCurve(FloatCurve=DataInterface() {propertyValues {Curve {keys [<at -0.0046 1>,<at 1.0011 4>]}}})
    Index 0 Param 2=&FloatFromCurve(FloatCurve=DataInterface() {propertyValues {Curve {keys [<at -0.0046 0>,<at 1.0011 2>]}}})
    Index 0 Param 3=&FloatFromCurve(FloatCurve=DataInterface() {propertyValues {Curve {keys [<at -0.0046 1>,<at 1.0011 0.5>]}}})
    Index 1 Param 3=512
    Index 2 Param 1=&FloatFromCurve(FloatCurve=DataInterface() {propertyValues {Curve {keys [<at -0.0046 0.4>,<at 1.0011 1.5>]}}})
    Index 2 Param 2=&FloatFromCurve(FloatCurve=DataInterface() {propertyValues {Curve {keys [<at -0.0046 0.1>, {interpMode=RCIM_Cubic, tangentMode=RCTM_User, time=0.11, value=2.6}, <at 1.0011 0.3>]}}})
    Index 2 Param 3=1
    Index 2 Param 4=&FloatFromCurve(FloatCurve=DataInterface() {propertyValues {Curve {keys [
         {arriveTangent=-1.9991, interpMode=RCIM_Cubic, time=0, value=1}
         {arriveTangent=-2.7143, interpMode=RCIM_Cubic, leaveTangent=-2.7143, tangentMode=RCTM_User, time=0.2095, value=0.6901}
         <at 1 0 RCIM_Cubic>
        ]}}})
    Module.Param0WriteEnabled=true
    Module.Param0WriteEnabled1=true
    Module.Param0WriteEnabled2=true
    Module.Param1WriteEnabled=true
    Module.Param1WriteEnabled1=true
    Module.Param1WriteEnabled2=true
    Module.Param2WriteEnabled=true
    Module.Param2WriteEnabled1=true
    Module.Param2WriteEnabled2=true
    Module.Param3WriteEnabled=true
    Module.Param3WriteEnabled1=true
    Module.Param3WriteEnabled2=true
    Write Parameter Index 1=true
    Write Parameter Index 2=true
   ) {enabled=false, module=DynamicMaterialParameters001}
   &DynamicMaterialParameters(
    Index 0 Param 1=&Add_Float(
     A=&RandomRangeFloat(Minimum=-2, Maximum=2)
     B=&FloatFromCurve(FloatCurve=DataInterface() {propertyValues {Curve {keys [<at 0 1>,<at 1 0.5>]}}}, Scale Curve=2)
    )
    Index 0 Param 2=&FloatFromCurve(FloatCurve=DataInterface() {propertyValues {Curve {keys [<at -0.0046 0>,<at 1.0011 2>]}}})
    Index 0 Param 3=&FloatFromCurve(FloatCurve=DataInterface() {propertyValues {Curve {keys [<at -0.0046 1>,<at 1.0011 0.5>]}}})
    Index 1 Param 1=&MakeFloatFromVector(Vector=/Niagara/DynamicInputs/Helpers/SimulationPosition(), Channel=::NewEnumerator0)
    Index 1 Param 2=&MakeFloatFromVector(Vector=&SimulationPosition(), Channel=::NewEnumerator1)
    Index 1 Param 3=&FloatFromCurve(
     FloatCurve=DataInterface() {propertyValues {Curve {keys [<at 0 0.8>, {arriveTangent=0.579, leaveTangent=0.579, time=1, value=1}]}}}
     Scale Curve=1600
    )
    Index 2 Param 1=&FloatFromCurve(FloatCurve=DataInterface() {propertyValues {Curve {keys [
         {interpMode=RCIM_Cubic, tangentMode=RCTM_User, time=0, value=2}
         {interpMode=RCIM_Cubic, tangentMode=RCTM_User, time=0.5, value=0.4}
         {interpMode=RCIM_Cubic, tangentMode=RCTM_User, time=1, value=2}
        ]}}})
    Index 2 Param 2=&FloatFromCurve(FloatCurve=DataInterface() {propertyValues {Curve {keys [
         {interpMode=RCIM_Cubic, tangentMode=RCTM_User, time=0, value=0.125}
         {interpMode=RCIM_Cubic, tangentMode=RCTM_User, time=0.25, value=2.75}
         <at 1.0011 0.3>
        ]}}})
    Index 2 Param 3=1
    Index 2 Param 4=&FloatFromCurve(
     FloatCurve=DataInterface() {propertyValues {Curve {keys [
         {arriveTangent=-1.9991, interpMode=RCIM_Cubic, time=0, value=1}
         {arriveTangent=-2.7143, interpMode=RCIM_Cubic, leaveTangent=-2.7143, tangentMode=RCTM_User, time=0.2095, value=0.6901}
         <at 1 0 RCIM_Cubic>
        ]}}}
     Scale Curve=2
    )
    Module.Param0WriteEnabled=true
    Module.Param0WriteEnabled1=true
    Module.Param0WriteEnabled2=true
    Module.Param1WriteEnabled=true
    Module.Param1WriteEnabled1=true
    Module.Param1WriteEnabled2=true
    Module.Param2WriteEnabled=true
    Module.Param2WriteEnabled1=true
    Module.Param2WriteEnabled2=true
    Module.Param3WriteEnabled=true
    Module.Param3WriteEnabled1=false
    Module.Param3WriteEnabled2=true
    Write Parameter Index 1=true
    Write Parameter Index 2=true
   ) {module=DynamicMaterialParameters002}
   &ScaleMeshSize(Scale Factor=&VectorFromFloat(Value=&FloatFromCurve(FloatCurve=DataInterface() {propertyValues {Curve {keys [<at -0.0046 1>,<at 1.0011 2.6>]}}}))) {module=MeshSizeScale}
   &ScaleColor(
    Scale Alpha=&Multiply_Float(
     A=1
     B=/Niagara/DynamicInputs/Scalability/ScalabilityDistanceBasedFloat(
      Curve:NiagaraDataInterfaceCurve=DataInterface() {propertyValues {Curve {keys [
          {interpMode=RCIM_Cubic, tangentMode=RCTM_User, time=0.75, value=1}
          {arriveTangent=-10.0464, interpMode=RCIM_Cubic, leaveTangent=-10.0464, tangentMode=RCTM_User, time=1, value=0}
         ]}}}
      CurveIndex:NiagaraFloat=$Emitter.Scalability.DistanceFraction
      Bias:NiagaraFloat=0
      Scale:NiagaraFloat=1
     )
    )
    Scale RGB=&Multiply_VectorByFloat(
     Vector=&MakeVectorFromLinearColorRGB(Color=&ColorFromCurve(
       ColorCurve=DataInterface() {propertyValues {
         AlphaCurve {keys [<at 0 1>,<at 1 1>]}
         BlueCurve {keys [
           {arriveTangent=0.256, arriveTangentWeight=0.0864, interpMode=RCIM_Cubic, leaveTangentWeight=0.0864, time=0.001, value=2}
           <at 0.0233 0.8542>
           <at 0.1 0.25>
           <at 0.14 0.9271>
           <at 0.2913 0.6382>
           {
            arriveTangent=0.246
            arriveTangentWeight=0.0709
            interpMode=RCIM_Cubic
            leaveTangent=0.246
            leaveTangentWeight=0.0709
            tangentMode=RCTM_User
            tangentWeightMode=RCTWM_WeightedBoth
            time=0.5
            value=0.4948
           }
           <at 0.839 0.6382>
           <at 0.89 0.6474>
           {arriveTangent=12.2964, interpMode=RCIM_Cubic, leaveTangent=12.2964, time=0.89, value=0.5677}
           {arriveTangentWeight=0.1844, interpMode=RCIM_Cubic, leaveTangent=0.0005, leaveTangentWeight=0.1844, time=1, value=2}
          ]}
         GreenCurve {keys [
           {arriveTangent=0.2383, arriveTangentWeight=0.0668, interpMode=RCIM_Cubic, leaveTangentWeight=0.0668, time=0.001, value=0.7367}
           <at 0.0233 0.8245>
           <at 0.1 0.1849>
           <at 0.14 0.8595>
           <at 0.2913 0.51>
           {
            arriveTangent=0.0845
            arriveTangentWeight=0.0575
            interpMode=RCIM_Cubic
            leaveTangent=0.0845
            leaveTangentWeight=0.0575
            tangentMode=RCTM_User
            tangentWeightMode=RCTWM_WeightedBoth
            time=0.5
            value=0.4694
           }
           <at 0.839 0.51>
           <at 0.89 0.5698>
           {
            arriveTangent=0.6573
            arriveTangentWeight=0.3221
            interpMode=RCIM_Cubic
            leaveTangent=0.6573
            leaveTangentWeight=0.3221
            tangentMode=RCTM_User
            tangentWeightMode=RCTWM_WeightedBoth
            time=1
            value=0.7367
           }
          ]}
         RedCurve {keys [
           {arriveTangent=0.2622, arriveTangentWeight=0.0786, interpMode=RCIM_Cubic, leaveTangentWeight=0.0786, time=0.001, value=0.6363}
           <at 0.0233 0.7741>
           <at 0.1 0.2485>
           <at 0.14 0.8614>
           <at 0.2913 0.6562>
           {
            arriveTangent=0.0426
            arriveTangentWeight=0.1136
            interpMode=RCIM_Cubic
            leaveTangent=0.0426
            leaveTangentWeight=0.1136
            tangentMode=RCTM_User
            tangentWeightMode=RCTWM_WeightedBoth
            time=0.5
            value=0.469
           }
           <at 0.839 0.6562>
           <at 0.89 0.599>
           {arriveTangent=0.3385, interpMode=RCIM_Cubic, leaveTangent=0.3385, time=0.89, value=0.7971}
           {arriveTangentWeight=0.3333, interpMode=RCIM_Cubic, tangentMode=RCTM_User, tangentWeightMode=RCTWM_WeightedBoth, time=1, value=0.6363}
          ]}
        }}
       CurveIndex=Linked() {linkedVariable {name=NPC.FortniteNPC.TimeOfDay_0-1, type=NiagaraFloat}}
      ))
     Float=1.75
    )
    ScaleA=true
    ScaleRGB=true
   )
   &ConstrainPositionToPlane(Plane Normal:Vector3f=$Particles.WaterPlaneNormal, Plane Position:Vector3f=$Particles.WaterPlaneLocation)
   &ScaleColor(
    Scale Alpha=&Add_Float(
     A=&FloatFromCurve(
      FloatCurve=DataInterface() {propertyValues {Curve {keys [<at 1024 0>,<at 2048 0.5>]}}}
      CurveIndex=$Engine.Owner.LODDistance
      Scale Curve=&FN_PickFloatDependOnPlatform(
       TargetPlatform=DataInterface() {propertyValues {Platforms {
          qualityLevelMask=-31
          deviceProfileStates [
           {profileName=Mobile, qualityLevelMask=4294967295, setQualityLevelMask=4294967295}
           {profileName=Switch, qualityLevelMask=4294967295, setQualityLevelMask=4294967295}
          ]
         }}}
       TargetPlatformActive=0
       TargetPlatformNoActive=1
      )
     )
     B=1
    )
    ScaleA=true
    ScaleRGB=true
   ) {module=ScaleColor002}
   &ScaleColor(
    Scale Alpha=&FN_PickFloatDependOnPlatform(
     TargetPlatform=DataInterface() {propertyValues {Platforms {
        qualityLevelMask=-31
        deviceProfileStates [
         {profileName=Switch, qualityLevelMask=4294967295, setQualityLevelMask=4294967295}
         {profileName=Mobile, qualityLevelMask=4294967295, setQualityLevelMask=4294967295}
        ]
       }}}
     TargetPlatformActive=2.5
     TargetPlatformNoActive=1
    )
    ScaleA=true
    ScaleRGB=true
   ) {module=ScaleColor003}
  ]
  rendererClasses [&NiagaraMeshRendererProperties]
  renderers [{rendererClass=&NiagaraMeshRendererProperties, rendererIndex=0}]
 }
}
summary {
 systemName=NS_Flopper_Spot_Creative
 emitters [
  {emitterName=Bubbles, rendererClasses [&NiagaraMeshRendererProperties]}
  {emitterName=Rings, rendererClasses [&NiagaraMeshRendererProperties]}
  {bEnabled=false, emitterName=Fish, rendererClasses [&NiagaraMeshRendererProperties]}
  {bEnabled=false, emitterName=Fish_Effective, rendererClasses [&NiagaraMeshRendererProperties]}
  {bEnabled=false, emitterName=Fish_Less_Effective, rendererClasses [&NiagaraMeshRendererProperties]}
  {bEnabled=false, emitterName=Fish_RareFishes, rendererClasses [&NiagaraMeshRendererProperties]}
  {bEnabled=false, emitterName=Fish_AllInOne, rendererClasses [&NiagaraMeshRendererProperties]}
  {bEnabled=false, emitterName=Fish001, rendererClasses [&NiagaraMeshRendererProperties]}
  {emitterName=Fish003, rendererClasses [&NiagaraMeshRendererProperties]}
 ]
 userVariables [
  User.FishCount {type=NiagaraInt32, defaultValue=3}
  User.RippleScaling {type=NiagaraFloat, defaultValue=0, description=""}
  User.RippleSpawnRate {type=NiagaraFloat, defaultValue=1, description=""}
  User.WaterBodyIndex {type=NiagaraFloat, defaultValue=0, description=""}
  User.WaterDepth {type=NiagaraFloat, defaultValue=0, description=""}
  User.WaterPlaneLocation {type=Vector3f, description="", defaultValue {struct=Vector3f, value=<0 0 0>}}
  User.WaterPlaneNormal {type=Vector3f, description="", defaultValue {struct=Vector3f, value=<0 0 0>}}
  User.WaveAttenuationFactor {type=NiagaraFloat, defaultValue=0, description=""}
  User.WaveTime {type=NiagaraFloat, defaultValue=0, description=""}
 ]
}
