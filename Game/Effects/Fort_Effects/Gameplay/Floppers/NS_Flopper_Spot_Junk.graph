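# Niagara system for a junk flopper spot in water, showing orbiting fish, bubbles and surface rings, all as mesh particles.
# Bubbles spawn continuously below the origin and rise while orbiting; Fish_Effective bursts three circling fish, while the other two fish emitters are disabled.
# Rings grow over their lifetime, constrained to the water plane, passing user water body index and depth into dynamic material parameters.
# Curve-driven material parameters animate bubbles and rings.
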NiagaraSystem /Game/Effects/Fort_Effects/Gameplay/Floppers/NS_Flopper_Spot_Junk
dependencies {
 usedDataInterfaces [&NiagaraDataInterfaceCurve]
 usedDynamicInputs []
 usedModules [/Niagara/Modules/System/SystemLifeCycle, &SystemState]
}
emitters {
 Bubbles {
  emitterUpdateScript [
   &EmitterState()
   &EmitterScalability(
    MaxDistance=600
    Module.bApplySpawnCountScale=true
    Module.bApplySpawnCountScaleByDistanceFraction=true
    Module.bUseMaxDistance=true
    Module.bUseMinDistance=false
    SpawnCountScaleByDistanceFraction=DataInterface() {propertyValues {Curve {keys [<at 0 1>,<at 1 0.5>]}}}
   ) {enabled=false}
   &EmitterState() {module=EmitterState001}
   &EmitterLifeCycle() {enabled=false}
   &SpawnRate(SpawnRate=5, Use Spawn Probability=false)
  ]
  particleSpawnScript [
   /Game/Effects/Niagara/Modules/Orbit/FN_Orbit_Init_Variables_v2()
   /Niagara/Modules/Spawn/Initialization/InitializeParticle(
    Lifetime=&UniformRangedFloat(Minimum=1.5, Maximum=3)
    Mesh Scale=&UniformRangedVector(Minimum=<2 2 1>, Maximum=<5 5 1>)
    Module.Write Color=true
    Module.Write Lifetime=true
    Module.Write Mass=true
    Module.Write Material Random=true
    Module.Write Position=false
    Module.Write Scale=true
    Module.Write SpriteRotation=false
    Module.Write SpriteSize=false
    Position=&SimulationPosition()
   )
   /Niagara/Modules/Spawn/Orientation/V2/InitialMeshOrientation(
    Orientation Vector=/Niagara/DynamicInputs/Random/RandomVector(
     Vector Scale:NiagaraFloat=1
     Constraint Vector:Vector3f=<0 0 1>
     Randomness Mode:ENiagaraRandomnessMode=::NewEnumerator0
     Random Seed:NiagaraInt32=0
     Fixed Random Seed:NiagaraBool=false
    )
    Use Orientation Vector=true
    Use Rotation Vector=false
   )
   &AddVelocity()
   &SystemLocation(MaskSpawnGroup=false, Offset=<0 0 -150>)
   /Niagara/Modules/Spawn/Location/SphereLocation(
    MaskSpawnGroup=false
    Module.Override Randomness=false
    Non Uniform Scale=<1.5 1.5 0.01>
    Offset=<0 0 -50>
    Sphere Radius=150
    Surface Only=false
   )
  ]
  particleUpdateScript [
   &ParticleState()
   /Game/Effects/Niagara/Modules/Orbit/FN_Orbit_v4(Rotation Rate min/max:Vector2f=<-0.2 0.2>, To Axis:Vector3f=<1 0 0>)
   /Game/Effects/Niagara/Modules/Orbit/FN_Orbit_v3(Rotation Rate min/max:Vector2f=<-0.2 0.2>, To Axis:Vector3f=<1 0 0>) {enabled=false}
   &Drag(Use Linear Drag=true, Use Rotational Drag=false)
   &GravityForce(Gravity=<0 0 100>)
   &SolveForcesAndVelocity(Clamp Velocity=false, Limit Acceleration=false)
   &DynamicMaterialParameters(
    Index 0 Param 1=&FloatFromCurve(FloatCurve=DataInterface() {propertyValues {Curve {keys [{arriveTangent=-4.0769, time=-0.0004, value=0.9057}, <at 1.0011 0.002>]}}})
    Index 0 Param 2=&FloatFromCurve(FloatCurve=DataInterface() {propertyValues {Curve {keys [<at -0.0046 3.0507>,<at 1.0011 1>]}}})
    Index 0 Param 3=&FloatFromCurve(FloatCurve=DataInterface() {propertyValues {Curve {keys [
         {interpMode=RCIM_Cubic, tangentMode=RCTM_User, time=0, value=30}
         {arriveTangent=22.4428, interpMode=RCIM_Cubic, leaveTangent=22.4428, tangentMode=RCTM_User, time=0.4752, value=45.6649}
         <at 0.5 46.2>
         <at 1 51.4>
         {interpMode=RCIM_Cubic, tangentMode=RCTM_User, time=1.0024, value=51.9752}
         <at 1.2 52>
         <at 1.5 52>
        ]}}})
    Index 0 Param 4=&FloatFromCurve(FloatCurve=DataInterface() {propertyValues {Curve {keys [
         <at 0 0 RCIM_Cubic>
         {arriveTangent=0.002, interpMode=RCIM_Cubic, leaveTangent=0.002, time=0.1245, value=34.51}
         <at 1.0011 0.002 RCIM_Cubic>
        ]}}})
    Module.Param0WriteEnabled=true
    Module.Param1WriteEnabled=true
    Module.Param2WriteEnabled=true
    Module.Param3WriteEnabled=true
   )
   &ScaleColor(
    Scale Alpha=&FloatFromCurve(FloatCurve=DataInterface() {propertyValues {Curve {keys [
         {arriveTangent=2.836, interpMode=RCIM_Cubic, leaveTangent=2.836, tangentMode=RCTM_User, time=0, value=-1}
         {arriveTangent=1.5, interpMode=RCIM_Cubic, leaveTangent=1.5, tangentMode=RCTM_User, time=0.3, value=-0.3}
         <at 0.5 -0.1>
         {arriveTangent=-0.8689, interpMode=RCIM_Cubic, leaveTangent=-0.8689, tangentMode=RCTM_User, time=0.6, value=-0.1}
         <at 1 -1>
         <at 1.0011 -0.9961>
         <at 1.2 -1>
         <at 1.5 -1>
        ]}}})
    ScaleA=true
    ScaleRGB=true
   )
   &ScaleMeshSize(Scale Factor=&VectorFromFloat(Value=&FloatFromCurve(FloatCurve=DataInterface() {propertyValues {Curve {keys [
          {interpMode=RCIM_Cubic, tangentMode=RCTM_User, time=-0.0038, value=0.4623}
          {arriveTangent=1.0062, interpMode=RCIM_Cubic, leaveTangent=1.0062, tangentMode=RCTM_User, time=0.1683, value=0.5805}
          {arriveTangent=-0.1436, interpMode=RCIM_Cubic, leaveTangent=-0.1436, tangentMode=RCTM_User, time=0.543, value=0.7431}
          {arriveTangent=-1.6175, interpMode=RCIM_Cubic, leaveTangent=-1.6175, tangentMode=RCTM_User, time=0.786, value=0.4919}
          <at 1.0011 0.002>
         ]}}})))
   /Niagara/Modules/Update/Orientation/V2/UpdateMeshOrientation(
    Rotation Rate:NiagaraFloat=&UniformRangedFloat(Minimum=0.1, Maximum=0.25)
    Rotation Vector:Vector3f=&UniformRangedVector(Minimum=<0 0 0.1>, Maximum=<0 0 0.25>)
   )
  ]
  rendererClasses [&NiagaraMeshRendererProperties]
  renderers [{rendererClass=&NiagaraMeshRendererProperties, rendererIndex=0}]
 }
 Fish {
  bEnabled=false
  emitterUpdateScript [
   &EmitterScalability(
    Module.bApplySpawnCountScale=false
    Module.bApplySpawnCountScaleByDistanceFraction=false
    Module.bUseMaxDistance=false
    Module.bUseMinDistance=false
   ) {enabled=false}
   &EmitterLifeCycle() {enabled=false}
   &EmitterState(Life Cycle Mode=::NewEnumerator1, Loop Behavior=::NewEnumerator1, Loop Duration Mode=::NewEnumerator1, UseLoopDelay=false)
   &SpawnBurst_Instantaneous(Spawn Count=3, Use Spawn Probability=false)
  ]
  particleSpawnScript [
   &FN_Orbit_Init_Variables_v2()
   &InitializeParticle(
    Module.Write Color=true
    Module.Write Lifetime=true
    Module.Write Mass=true
    Module.Write Material Random=true
    Module.Write Position=false
    Module.Write Scale=true
    Module.Write SpriteRotation=false
    Module.Write SpriteSize=false
    Position=&SimulationPosition()
   )
   &AddRotationalVelocity(
    Coordinate Space:ENiagaraMeshTransforms=::NewEnumerator1
    Rotation Rate=&UniformRangedVector(Minimum=<-0.25 0 0>, Maximum=<0.25 0 0>)
   ) {enabled=false}
   &ApplyInitialForces() {enabled=false}
   &InitialMeshOrientation(Use Orientation Vector=true, Use Rotation Vector=true) {enabled=false}
   &SystemLocation(MaskSpawnGroup=false, Offset=&UniformRangedVector(Minimum=<-25 0 -165>, Maximum=<25 1 -185>))
   &SphereLocation(MaskSpawnGroup=false, Module.Override Randomness=false, Non Uniform Scale=<1 1 0.2>, Sphere Radius=25, Surface Only=false)
  ]
  particleUpdateScript [
   &FN_Orbit_v3(
    Height min/max:Vector2f=<0 150>
    Radius min/max:Vector2f=<125 150>
    Rotation Offset min/max:Vector2f=<1 5>
    Rotation Rate min/max:Vector2f=<0.1 0.2>
    To Axis:Vector3f=&Lerp_Vector(A=<-0.25 0 1>, B=<0.25 0.25 1>, Alpha=0.5)
   )
   &SolveForcesAndVelocity(Clamp Velocity=false, Limit Acceleration=false)
   &ParticleState(Kill Particles When Lifetime Has Elapsed=false, Let Infinitely Lived Particles Die When Emitter Deactivates=true)
   &Color()
  ]
  rendererClasses [&NiagaraMeshRendererProperties]
  renderers [{rendererClass=&NiagaraMeshRendererProperties, rendererIndex=0}]
 }
 Fish_Effective {
  emitterUpdateScript [
   &EmitterScalability(
    Module.bApplySpawnCountScale=false
    Module.bApplySpawnCountScaleByDistanceFraction=false
    Module.bUseMaxDistance=false
    Module.bUseMinDistance=false
   ) {enabled=false}
   &EmitterLifeCycle() {enabled=false}
   &EmitterState(Life Cycle Mode=::NewEnumerator1, Loop Behavior=::NewEnumerator1, Loop Duration Mode=::NewEnumerator1, UseLoopDelay=false)
   &SpawnBurst_Instantaneous(Spawn Count=3, Use Spawn Probability=false)
  ]
  particleSpawnScript [
   &FN_Orbit_Init_Variables_v2()
   &InitializeParticle(
    Mesh Scale=<1.25 1.25 1.25>
    Module.Write Color=true
    Module.Write Lifetime=true
    Module.Write Mass=true
    Module.Write Material Random=true
    Module.Write Position=false
    Module.Write Scale=true
    Module.Write SpriteRotation=false
    Module.Write SpriteSize=false
    Position=&SimulationPosition()
   )
   &AddRotationalVelocity(
    Coordinate Space:ENiagaraMeshTransforms=::NewEnumerator1
    Rotation Rate=&UniformRangedVector(Minimum=<-0.25 0 0>, Maximum=<0.25 0 0>)
   ) {enabled=false}
   &ApplyInitialForces()
   &InitialMeshOrientation(Use Orientation Vector=true, Use Rotation Vector=true) {enabled=false}
   &SystemLocation(MaskSpawnGroup=false, Offset=&UniformRangedVector(Minimum=<0 0 -165>, Maximum=<1 0 -180>))
   &SphereLocation(MaskSpawnGroup=false, Module.Override Randomness=false, Non Uniform Scale=<1 1 0.2>, Sphere Radius=25, Surface Only=false)
  ]
  particleUpdateScript [
   &FN_Orbit_v3(
    Height min/max:Vector2f=<0 150>
    Radius min/max:Vector2f=<165 200>
    Rotation Offset min/max:Vector2f=<1 5>
    Rotation Rate min/max:Vector2f=<0.1 0.2>
    To Axis:Vector3f=&Lerp_Vector(A=<0 0 1>, B=<0 0 1>, Alpha=0.5)
   ) {enabled=false}
   &FN_Orbit_v4(
    Height min/max:Vector2f=<0 150>
    Radius min/max:Vector2f=<165 200>
    Rotation Offset min/max:Vector2f=<1 5>
    Rotation Rate min/max:Vector2f=<0.1 0.2>
   )
   &SolveForcesAndVelocity(Clamp Velocity=false, Limit Acceleration=false)
   &ParticleState(Kill Particles When Lifetime Has Elapsed=false, Let Infinitely Lived Particles Die When Emitter Deactivates=true)
   &Color()
  ]
  rendererClasses [&NiagaraMeshRendererProperties]
  renderers [{rendererClass=&NiagaraMeshRendererProperties, rendererIndex=0}]
 }
 Fish_Less_Effective {
  bEnabled=false
  emitterUpdateScript [
   &EmitterScalability(
    Module.bApplySpawnCountScale=false
    Module.bApplySpawnCountScaleByDistanceFraction=false
    Module.bUseMaxDistance=false
    Module.bUseMinDistance=false
   ) {enabled=false}
   &EmitterLifeCycle() {enabled=false}
   &EmitterState(Life Cycle Mode=::NewEnumerator1, Loop Behavior=::NewEnumerator1, Loop Duration Mode=::NewEnumerator1, UseLoopDelay=false)
   &SpawnBurst_Instantaneous(Spawn Count=3, Use Spawn Probability=false)
  ]
  particleSpawnScript [
   &FN_Orbit_Init_Variables_v2()
   &InitializeParticle(
    Mesh Scale=<10 10 10>
    Module.Write Color=true
    Module.Write Lifetime=true
    Module.Write Mass=true
    Module.Write Material Random=true
    Module.Write Position=false
    Module.Write Scale=true
    Module.Write SpriteRotation=false
    Module.Write SpriteSize=false
    Position=&SimulationPosition()
   )
   &AddRotationalVelocity(
    Coordinate Space:ENiagaraMeshTransforms=::NewEnumerator1
    Rotation Rate=&UniformRangedVector(Minimum=<-0.25 0 0>, Maximum=<0.25 0 0>)
   ) {enabled=false}
   &ApplyInitialForces() {enabled=false}
   &InitialMeshOrientation(Use Orientation Vector=true, Use Rotation Vector=true) {enabled=false}
   &SystemLocation(MaskSpawnGroup=false, Offset=&UniformRangedVector(Minimum=<0 0 -180>, Maximum=<1 0 -220>))
   &SphereLocation(MaskSpawnGroup=false, Module.Override Randomness=false, Non Uniform Scale=<1 1 0.2>, Sphere Radius=25, Surface Only=false)
  ]
  particleUpdateScript [
   &FN_Orbit_v3(
    Height min/max:Vector2f=<0 150>
    Radius min/max:Vector2f=<200 300>
    Rotation Offset min/max:Vector2f=<1 5>
    Rotation Rate min/max:Vector2f=<0.1 0.2>
   )
   &SolveForcesAndVelocity(Clamp Velocity=false, Limit Acceleration=false)
   &ParticleState(Kill Particles When Lifetime Has Elapsed=false, Let Infinitely Lived Particles Die When Emitter Deactivates=true)
   &Color()
  ]
  rendererClasses [&NiagaraMeshRendererProperties]
  renderers [{rendererClass=&NiagaraMeshRendererProperties, rendererIndex=0}]
 }
 Rings {
  emitterUpdateScript [
   &EmitterState()
   &EmitterScalability(
    MaxDistance=2000
    Module.bApplySpawnCountScale=true
    Module.bApplySpawnCountScaleByDistanceFraction=true
    Module.bUseMaxDistance=true
    Module.bUseMinDistance=false
    SpawnCountScaleByDistanceFraction=DataInterface() {propertyValues {Curve {keys [<at 0 1>,<at 1 0.5>]}}}
   ) {enabled=false}
   &EmitterState() {module=EmitterState001}
   &SpawnPerUnit(Module.Use Max Movement Threshold=true, Module.Use Movement Tolerance=false, Spawn Per Unit=50, Use Spawn Probability=false) {enabled=false}
   &EmitterLifeCycle() {enabled=false}
   &SpawnBurst_Instantaneous(Spawn Count=1, Use Spawn Probability=false)
   &SpawnRate(Use Spawn Probability=false)
  ]
  particleSpawnScript [
   &InitializeParticle(
    Color=<rgba 2 2 2 1>
    Lifetime=&UniformRangedFloat(Minimum=4, Maximum=5)
    Mass=&UniformRangedFloat(Minimum=0.6, Maximum=1.2)
    Mesh Scale=<500 500 500>
    Module.Write Color=true
    Module.Write Lifetime=true
    Module.Write Mass=true
    Module.Write Material Random=true
    Module.Write Position=false
    Module.Write Scale=true
    Module.Write SpriteRotation=true
    Module.Write SpriteSize=false
    Position=&SimulationPosition()
    Sprite Rotation=&NormalizedAngleToDegrees(NormalizedAngle=&UniformRangedFloat(Minimum=0, Maximum=1))
   ) {enabled=false}
   /Niagara/Modules/Spawn/Initialization/V2/InitializeParticle(
    Lifetime=&UniformRangedFloat(Minimum=2, Maximum=3)
    Mass=&UniformRangedFloat(Minimum=0.6, Maximum=1.2)
    Mass Mode=::NewEnumerator1
    Mesh Scale Mode=::NewEnumerator3
    Mesh Uniform Scale=100
    Sprite Rotation Angle=&UniformRangedFloat(Minimum=0, Maximum=1)
    Sprite Rotation Mode=::NewEnumerator1
    UsePositionOffset=false
    Write Material Random=true
   ) {module=InitializeParticle001}
   &InitialMeshOrientation(
    Orientation Vector=<0 0 0>
    Rotation=&UniformRangedVector(Minimum=<0 0 0>, Maximum=<0 0 1>)
    Use Orientation Vector=true
    Use Rotation Vector=true
   ) {enabled=false}
   /Niagara/Modules/Spawn/Orientation/V3/InitialMeshOrientation(
    Mesh Orientation Mode=::NewEnumerator2
    Orientation Vector=<0 0 0>
    Rotation=&UniformRangedVector(Minimum=<0 0 0>, Maximum=<0 0 1>)
    Use Rotation Vector=true
   ) {module=InitialMeshOrientation001}
   &DynamicMaterialParameters(
    Index 1 Param 3=200
    Index 3 Param 1=$User.WaterBodyIndex
    Index 3 Param 2=$User.WaterDepth
    Module.Param0WriteEnabled1=false
    Module.Param0WriteEnabled3=true
    Module.Param1WriteEnabled1=false
    Module.Param1WriteEnabled3=true
    Module.Param2WriteEnabled1=true
    Module.Param2WriteEnabled3=true
    Module.Param3WriteEnabled1=false
    Module.Param3WriteEnabled3=true
    Write Parameter Index 0=false
    Write Parameter Index 1=true
    Write Parameter Index 3=true
   )
   &SystemLocation(MaskSpawnGroup=false, Offset=<0 0 5>)
   &InitializeWaterSnapToOwner()
   &InheritVelocity(Module.Limit Speed=true)
  ]
  particleUpdateScript [
   &ParticleState()
   &DynamicMaterialParameters(
    Index 0 Param 1=&FloatFromCurve(FloatCurve=DataInterface() {propertyValues {Curve {keys [<at -0.0046 1>,<at 1.0011 4>]}}})
    Index 0 Param 2=&FloatFromCurve(FloatCurve=DataInterface() {propertyValues {Curve {keys [<at -0.0046 0>,<at 1.0011 2>]}}})
    Index 0 Param 3=&FloatFromCurve(FloatCurve=DataInterface() {propertyValues {Curve {keys [<at -0.0046 1>,<at 1.0011 0.5>]}}})
    Index 1 Param 3=512
    Index 2 Param 1=&FloatFromCurve(FloatCurve=DataInterface() {propertyValues {Curve {keys [<at -0.0046 0.4>,<at 1.0011 1.5>]}}})
    Index 2 Param 2=&FloatFromCurve(FloatCurve=DataInterface() {propertyValues {Curve {keys [<at -0.0046 0.1>,<at 0.11 2.6>,<at 1.0011 0.3>]}}})
    Index 2 Param 3=1
    Index 2 Param 4=&FloatFromCurve(FloatCurve=DataInterface() {propertyValues {Curve {keys [
         {arriveTangent=-1.9991, interpMode=RCIM_Cubic, time=0, value=1}
         {arriveTangent=-2.7143, interpMode=RCIM_Cubic, leaveTangent=-2.7143, tangentMode=RCTM_User, time=0.2095, value=0.6901}
         <at 1 0 RCIM_Cubic>
        ]}}})
    Module.Param0WriteEnabled=true
    Module.Param0WriteEnabled1=true
    Module.Param0WriteEnabled2=true
    Module.Param1WriteEnabled=true
    Module.Param1WriteEnabled1=true
    Module.Param1WriteEnabled2=true
    Module.Param2WriteEnabled=true
    Module.Param2WriteEnabled1=true
    Module.Param2WriteEnabled2=true
    Module.Param3WriteEnabled=true
    Module.Param3WriteEnabled1=true
    Module.Param3WriteEnabled2=true
    Write Parameter Index 1=true
    Write Parameter Index 2=true
   ) {module=DynamicMaterialParameters001}
   &ScaleColor(
    Scale Alpha=&FloatFromCurve(FloatCurve=DataInterface() {propertyValues {Curve {keys [<at -0.0046 1.0113 RCIM_Cubic>,<at 1.0011 0.002 RCIM_Cubic>]}}})
    ScaleA=true
    ScaleRGB=true
   ) {enabled=false}
   &Drag(Drag=2, Use Linear Drag=true, Use Rotational Drag=false)
   &SolveForcesAndVelocity(Clamp Velocity=false, Limit Acceleration=false)
   &ScaleMeshSize(Scale Factor=&VectorFromFloat(Value=&FloatFromCurve(FloatCurve=DataInterface() {propertyValues {Curve {keys [<at -0.0046 1>,<at 1.0011 2.6>]}}}))) {module=MeshSizeScale}
   &ConstrainPositionToPlane(Plane Normal:Vector3f=$Particles.WaterPlaneNormal, Plane Position:Vector3f=$Particles.WaterPlaneLocation)
  ]
  rendererClasses [&NiagaraMeshRendererProperties]
  renderers [{rendererClass=&NiagaraMeshRendererProperties, rendererIndex=0}]
 }
}
summary {
 systemName=NS_Flopper_Spot_Junk
 emitters [
  {emitterName=Bubbles, rendererClasses [&NiagaraMeshRendererProperties]}
  {emitterName=Rings, rendererClasses [&NiagaraMeshRendererProperties]}
  {bEnabled=false, emitterName=Fish, rendererClasses [&NiagaraMeshRendererProperties]}
  {emitterName=Fish_Effective, rendererClasses [&NiagaraMeshRendererProperties]}
  {bEnabled=false, emitterName=Fish_Less_Effective, rendererClasses [&NiagaraMeshRendererProperties]}
 ]
 userVariables [
  User.WaterBodyIndex {type=NiagaraFloat, defaultValue=0, description=""}
  User.WaterDepth {type=NiagaraFloat, defaultValue=0, description=""}
  User.WaterPlaneLocation {type=Vector3f, description="", defaultValue {struct=Vector3f, value=<0 0 0>}}
  User.WaterPlaneNormal {type=Vector3f, description="", defaultValue {struct=Vector3f, value=<0 0 0>}}
  User.WaveAttenuationFactor {type=NiagaraFloat, defaultValue=0, description=""}
  User.WaveTime {type=NiagaraFloat, defaultValue=0, description=""}
 ]
}
